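# Widget Blueprint for the "Press Any Key" settings prompt, built on the native FortSettingsPressAnyKey class.
# A dimmed, background-blurred border centers two inherited CommonVisibilitySwitchers: a title (key or button) and a cancel hint (Escape, gamepad keybind or touch).
# The gamepad hint uses a keybind rich text block to show the Gamepad_Special_Right binding inline.
# The event graph is empty, so switching logic is left to the parent class.
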WidgetBlueprint /FortniteCoreUI/Settings/Screens/PressAnyKey
graphs [EventGraph {links [], nodes []}]
variables []
widgets {
 info {inheritedWidgetCount=2, parentClass=/Script/FortniteUI.FortSettingsPressAnyKey, rootWidgetClass=&Border, widgetCount=10}
 widgets [
  Border_0 {widgetClassPath=&Border, widget=Border_0, properties {brushColor=<rgba 0 0 0 0.2>}}
  BackgroundBlur_0 {
   widgetClassPath=&BackgroundBlur
   parent=Border_0
   widget=BackgroundBlur_0
   properties {blurStrength=50.1587, horizontalAlignment=HAlign_Center, verticalAlignment=VAlign_Center}
   slot {class=BorderSlot}
  }
  VerticalBox_1 {
   widgetClassPath=&VerticalBox
   parent=BackgroundBlur_0
   widget=VerticalBox_1
   slot {class=BackgroundBlurSlot, properties {horizontalAlignment=HAlign_Center, verticalAlignment=VAlign_Center, padding=<ltrb 0 0 0 0>}}
  }
  Switcher_Title {
   widgetClassPath=&CommonVisibilitySwitcher
   bInherited=true
   parent=VerticalBox_1
   widget=Switcher_Title
   slot {class=VerticalBoxSlot, properties {horizontalAlignment=HAlign_Center}}
  }
  Text_PressAnyKey {
   widgetClassPath=&TextBlock
   parent=Switcher_Title
   widget=Text_PressAnyKey
   properties {
    text=Press Any Key
    font {
     bForceMonospaced=false
     bMaterialIsStencil=false
     fontMaterial=null
     fontObject=&BurbankSmall
     letterSpacing=0
     monospacedWidth=1
     size=60
     skewAmount=0
     typefaceFontName=Bold
     outlineSettings {
      bApplyOutlineToDropShadows=false
      bMiteredCorners=false
      bSeparateFillAlpha=false
      outlineMaterial=null
      outlineSize=0
      outlineColor=<rgba 0 0 0 1>
     }
    }
   }
   slot {class=CommonVisibilitySwitcherSlot}
  }
  Text_PressAnyButton {
   widgetClassPath=&TextBlock
   parent=Switcher_Title
   widget=Text_PressAnyButton
   properties {
    text=Press Any Button
    font {
     bForceMonospaced=false
     bMaterialIsStencil=false
     fontMaterial=null
     fontObject=&BurbankSmall
     letterSpacing=0
     monospacedWidth=1
     size=60
     skewAmount=0
     typefaceFontName=Bold
     outlineSettings {
      bApplyOutlineToDropShadows=false
      bMiteredCorners=false
      bSeparateFillAlpha=false
      outlineMaterial=null
      outlineSize=0
      outlineColor=<rgba 0 0 0 1>
     }
    }
   }
   slot {class=CommonVisibilitySwitcherSlot, properties {horizontalAlignment=HAlign_Center}}
  }
  Switcher_Cancel {
   widgetClassPath=&CommonVisibilitySwitcher
   bInherited=true
   parent=VerticalBox_1
   widget=Switcher_Cancel
   slot {class=VerticalBoxSlot, properties {horizontalAlignment=HAlign_Center}}
  }
  CommonTextBlock_0 {
   widgetClassPath=&CommonTextBlock
   parent=Switcher_Cancel
   widget=CommonTextBlock_0
   properties {
    justification=Center
    style=/FortUILibrary/Styles/Typography/Foundation/TextStyle-Base-M.TextStyle-Base-M_C
    text=Escape to Cancel
    font {
     bForceMonospaced=false
     bMaterialIsStencil=false
     fontMaterial=null
     fontObject=&NotoSans
     letterSpacing=0
     monospacedWidth=1
     size=24
     skewAmount=0
     typefaceFontName=NotoSans-Regular
     outlineSettings {
      bApplyOutlineToDropShadows=false
      bMiteredCorners=false
      bSeparateFillAlpha=false
      outlineMaterial=null
      outlineSize=0
      outlineColor=<rgba 0 0 0 1>
     }
    }
    shadowOffset=<0 0>
   }
   slot {class=CommonVisibilitySwitcherSlot}
  }
  FortRichTextBlockKeybind_60 {
   widgetClassPath=/Script/FortniteUI.FortRichTextBlockKeybind
   parent=Switcher_Cancel
   widget=FortRichTextBlockKeybind_60
   properties {
    bOverrideDefaultStyle=true
    defaultTextStyleOverrideClass=&TextStyle-Base_C
    text="Press <keybind_Raw>Gamepad_Special_Right</> to Cancel"
    defaultTextStyleOverride {
     overflowPolicy=Clip
     transformPolicy=None
     colorAndOpacity {colorUseRule=UseColor_Specified, specifiedColor=<rgba 1 1 1 1>}
     font {
      bForceMonospaced=false
      bMaterialIsStencil=false
      fontMaterial=null
      fontObject=&NotoSans
      letterSpacing=0
      monospacedWidth=1
      size=24
      skewAmount=0
      typefaceFontName=NotoSans-Regular
      outlineSettings {
       bApplyOutlineToDropShadows=false
       bMiteredCorners=false
       bSeparateFillAlpha=false
       outlineMaterial=null
       outlineSize=0
       outlineColor=<rgba 0 0 0 1>
      }
     }
     highlightColor {colorUseRule=UseColor_Specified, specifiedColor=<rgba 0 0 0 1>}
     highlightShape {
      bIsDynamicallyLoaded=false
      drawAs=Image
      imageType=NoImage
      mirroring=NoMirror
      resourceName=None
      resourceObject=null
      tiling=NoTile
      imageSize=<32 32>
      margin=<ltrb 0 0 0 0>
      outlineSettings {
       bUseBrushTransparency=false
       roundingType=HalfHeightRadius
       width=0
       color {colorUseRule=UseColor_Specified, specifiedColor=<rgba 0 0 0 0>}
       cornerRadii=<0 0 0 0>
      }
      tintColor {colorUseRule=UseColor_Specified, specifiedColor=<rgba 1 1 1 1>}
      uVRegion {bIsValid=false, max=<0 0>, min=<0 0>}
     }
     selectedBackgroundColor {colorUseRule=UseColor_Foreground, specifiedColor=<rgba 1 0 1 1>}
     shadowColorAndOpacity=<rgba 0 0 0 1>
     shadowOffset=<0 0>
     strikeBrush {
      bIsDynamicallyLoaded=false
      drawAs=Image
      imageType=NoImage
      mirroring=NoMirror
      resourceName=None
      resourceObject=null
      tiling=NoTile
      imageSize=<32 32>
      margin=<ltrb 0 0 0 0>
      outlineSettings {
       bUseBrushTransparency=false
       roundingType=HalfHeightRadius
       width=0
       color {colorUseRule=UseColor_Specified, specifiedColor=<rgba 0 0 0 0>}
       cornerRadii=<0 0 0 0>
      }
      tintColor {colorUseRule=UseColor_Specified, specifiedColor=<rgba 1 1 1 1>}
      uVRegion {bIsValid=false, max=<0 0>, min=<0 0>}
     }
     underlineBrush {
      bIsDynamicallyLoaded=false
      drawAs=Image
      imageType=NoImage
      mirroring=NoMirror
      resourceName=None
      resourceObject=null
      tiling=NoTile
      imageSize=<32 32>
      margin=<ltrb 0 0 0 0>
      outlineSettings {
       bUseBrushTransparency=false
       roundingType=HalfHeightRadius
       width=0
       color {colorUseRule=UseColor_Specified, specifiedColor=<rgba 0 0 0 0>}
       cornerRadii=<0 0 0 0>
      }
      tintColor {colorUseRule=UseColor_Specified, specifiedColor=<rgba 1 1 1 1>}
      uVRegion {bIsValid=false, max=<0 0>, min=<0 0>}
     }
    }
   }
   slot {class=CommonVisibilitySwitcherSlot, properties {horizontalAlignment=HAlign_Center}}
  }
  CommonTextBlock {
   widgetClassPath=&CommonTextBlock
   parent=Switcher_Cancel
   widget=CommonTextBlock
   properties {
    justification=Center
    style=&TextStyle-Base-M_C
    text=Touch to Cancel
    font {
     bForceMonospaced=false
     bMaterialIsStencil=false
     fontMaterial=null
     fontObject=&NotoSans
     letterSpacing=0
     monospacedWidth=1
     size=24
     skewAmount=0
     typefaceFontName=NotoSans-Regular
     outlineSettings {
      bApplyOutlineToDropShadows=false
      bMiteredCorners=false
      bSeparateFillAlpha=false
      outlineMaterial=null
      outlineSize=0
      outlineColor=<rgba 0 0 0 1>
     }
    }
    shadowOffset=<0 0>
   }
   slot {class=CommonVisibilitySwitcherSlot, properties {horizontalAlignment=HAlign_Center}}
  }
 ]
}
