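# Settings screen overlay that prompts the player to press a key on a MIDI controller, built on FortSettingsPressMIDIInput.
# A tinted border and background blur hold a centered title and a cancel hint.
# The inherited Switcher_Cancel holds three hint variants: Escape, a keybind widget with a gamepad fallback key, and touch.
# The event graph is empty, so input capture and hint switching are left to the parent class.
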WidgetBlueprint /FortniteCoreUI/Settings/Screens/PressMIDIKey
graphs [EventGraph {links [], nodes []}]
variables []
widgets {
 info {inheritedWidgetCount=1, parentClass=/Script/FortniteUI.FortSettingsPressMIDIInput, rootWidgetClass=&Border, widgetCount=12}
 widgets [
  Border_0 {widgetClassPath=&Border, widget=Border_0, properties {brushColor=<rgba 0 0 0 0.2>}}
  BackgroundBlur_0 {
   widgetClassPath=&BackgroundBlur
   parent=Border_0
   widget=BackgroundBlur_0
   properties {blurStrength=50.1587, horizontalAlignment=HAlign_Center, verticalAlignment=VAlign_Center}
   slot {class=BorderSlot}
  }
  VerticalBox_1 {
   widgetClassPath=&VerticalBox
   parent=BackgroundBlur_0
   widget=VerticalBox_1
   slot {class=BackgroundBlurSlot, properties {horizontalAlignment=HAlign_Center, verticalAlignment=VAlign_Center, padding=<ltrb 0 0 0 0>}}
  }
  Switcher_Title {
   widgetClassPath=&CommonVisibilitySwitcher
   parent=VerticalBox_1
   widget=Switcher_Title
   slot {class=VerticalBoxSlot, properties {horizontalAlignment=HAlign_Center}}
  }
  Text_PressAnyKey {
   widgetClassPath=&TextBlock
   parent=Switcher_Title
   widget=Text_PressAnyKey
   properties {
    text=PRESS ANY KEY ON YOUR MIDI CONTROLLER
    font {
     bForceMonospaced=false
     bMaterialIsStencil=false
     fontMaterial=null
     fontObject=&HeadingNow
     letterSpacing=0
     monospacedWidth=1
     size=25.5
     skewAmount=0
     typefaceFontName=86 Bold Italic
     outlineSettings {
      bApplyOutlineToDropShadows=false
      bMiteredCorners=false
      bSeparateFillAlpha=false
      outlineMaterial=null
      outlineSize=0
      outlineColor=<rgba 0 0 0 1>
     }
    }
   }
   slot {class=CommonVisibilitySwitcherSlot}
  }
  Switcher_Cancel {
   widgetClassPath=&CommonVisibilitySwitcher
   bInherited=true
   parent=VerticalBox_1
   widget=Switcher_Cancel
   slot {class=VerticalBoxSlot, properties {horizontalAlignment=HAlign_Center}}
  }
  CommonTextBlock_0 {
   widgetClassPath=&CommonTextBlock
   parent=Switcher_Cancel
   widget=CommonTextBlock_0
   properties {
    justification=Center
    text=ESCAPE TO CANCEL
    font {
     bForceMonospaced=false
     bMaterialIsStencil=false
     fontMaterial=null
     fontObject=&HeadingNow
     letterSpacing=0
     monospacedWidth=1
     size=18
     skewAmount=0
     typefaceFontName=74 Regular
     outlineSettings {
      bApplyOutlineToDropShadows=false
      bMiteredCorners=false
      bSeparateFillAlpha=false
      outlineMaterial=null
      outlineSize=0
      outlineColor=<rgba 0 0 0 1>
     }
    }
    shadowOffset=<0 0>
   }
   slot {class=CommonVisibilitySwitcherSlot}
  }
  HorizontalBox_0 {
   widgetClassPath=&HorizontalBox
   parent=Switcher_Cancel
   widget=HorizontalBox_0
   slot {class=CommonVisibilitySwitcherSlot, properties {horizontalAlignment=HAlign_Center}}
  }
  CommonTextBlock_1 {
   widgetClassPath=&CommonTextBlock
   parent=HorizontalBox_0
   widget=CommonTextBlock_1
   properties {
    justification=Center
    text=PRESS
    font {
     bForceMonospaced=false
     bMaterialIsStencil=false
     fontMaterial=null
     fontObject=&HeadingNow
     letterSpacing=0
     monospacedWidth=1
     size=18
     skewAmount=0
     typefaceFontName=74 Regular
     outlineSettings {
      bApplyOutlineToDropShadows=false
      bMiteredCorners=false
      bSeparateFillAlpha=false
      outlineMaterial=null
      outlineSize=0
      outlineColor=<rgba 0 0 0 1>
     }
    }
    shadowOffset=<0 0>
   }
   slot {class=HorizontalBoxSlot, properties {verticalAlignment=VAlign_Center}}
  }
  KeybindWidget {
   widgetClassPath=/FortUILibrary/Input/KeybindWidget
   bIsVariable=true
   parent=HorizontalBox_0
   widget=KeybindWidget
   properties {boundKeyFallback=Gamepad_Special_Right}
   slot {class=HorizontalBoxSlot, properties {verticalAlignment=VAlign_Center, padding=<ltrb 10 0 10 0>}}
  }
  CommonTextBlock_2 {
   widgetClassPath=&CommonTextBlock
   parent=HorizontalBox_0
   widget=CommonTextBlock_2
   properties {
    justification=Center
    text=TO CANCEL
    font {
     bForceMonospaced=false
     bMaterialIsStencil=false
     fontMaterial=null
     fontObject=&HeadingNow
     letterSpacing=0
     monospacedWidth=1
     size=18
     skewAmount=0
     typefaceFontName=74 Regular
     outlineSettings {
      bApplyOutlineToDropShadows=false
      bMiteredCorners=false
      bSeparateFillAlpha=false
      outlineMaterial=null
      outlineSize=0
      outlineColor=<rgba 0 0 0 1>
     }
    }
    shadowOffset=<0 0>
   }
   slot {class=HorizontalBoxSlot, properties {verticalAlignment=VAlign_Center}}
  }
  CommonTextBlock {
   widgetClassPath=&CommonTextBlock
   parent=Switcher_Cancel
   widget=CommonTextBlock
   properties {
    justification=Center
    text=TOUCH TO CANCEL
    font {
     bForceMonospaced=false
     bMaterialIsStencil=false
     fontMaterial=null
     fontObject=&HeadingNow
     letterSpacing=0
     monospacedWidth=1
     size=18
     skewAmount=0
     typefaceFontName=74 Regular
     outlineSettings {
      bApplyOutlineToDropShadows=false
      bMiteredCorners=false
      bSeparateFillAlpha=false
      outlineMaterial=null
      outlineSize=0
      outlineColor=<rgba 0 0 0 1>
     }
    }
    shadowOffset=<0 0>
   }
   slot {class=CommonVisibilitySwitcherSlot, properties {horizontalAlignment=HAlign_Center}}
  }
 ]
}
